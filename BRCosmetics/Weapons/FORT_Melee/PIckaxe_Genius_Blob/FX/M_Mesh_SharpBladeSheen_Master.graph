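# Additive unlit master material for mesh particles that sweeps a sheen band across a blade, here under the Genius Blob pickaxe cosmetic.
# UV1 is rotated and rescaled in the vertex stage, then remapped into a band driven by dynamic parameters, with a looping time fallback outside particle systems.
# A static switch picks procedural or gradient-texture sheen, tinted by particle and optional vertex color.
# Camera-direction and normal-push terms form a vertex offset.
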Material /BRCosmetics/Weapons/FORT_Melee/PIckaxe_Genius_Blob/FX/M_Mesh_SharpBladeSheen_Master
graph {
 properties {
  bUsedWithNiagaraMeshParticles=true
  blendMode=BLEND_Additive
  pixelDepthOffsetMode=PDOM_Legacy
  shadingModel=MSM_Unlit
  translucencyPass=MTP_BeforeDOF
  usedShadingModels=Unlit
 }
 nodes [
  n1 = Subtract(A=CameraPositionWS(), B=WorldPosition().XYZ)
  Add(
   A=Multiply(A=Normalize(VectorInput=@n1), B=ScalarParameter() {defaultValue=1, group=Vertex Offset, parameterName=Camera Offset, sliderMax=1})
   B=Multiply(A=VertexNormalWS(), B=ScalarParameter() {group=Vertex Offset, parameterName=Vertex Normal Push})
  )
  n2 = VertexColor()
  n3 = ParticleColor()
  n4 = Multiply(
   A=StaticSwitchParameter(True=AppendVector(A=@n2, B=@n2), False=Constant() {r=1}) {group=Sheen, parameterName=Use Vertex Colors, sortPriority=1}
   B=AppendVector(A=@n3.RGB, B=@n3.RGB)
  )
  ComponentMask(None=@n4) {a=true}
  n5 = MaterialFunctionCall(Float2=Subtract(A=TextureCoordinate() {coordinateIndex=1}) {constB=0.5}) {materialFunction=&BreakOutFloat2Components}
  n6 = DynamicParameter() {defaultValue=<rgba 0 0.333 0.03 0>, paramNames [Sheen Sweep, Sheen Size, Sheen Rotation, Sheen Time]}
  n7 = Cosine(None=@n6.Sheen Rotation)
  n8 = Sine(None=@n6.Sheen Rotation)
  n9 = Add(A=Multiply(A=@n5.R, B=@n7), B=Multiply(A=@n5.G, B=@n8))
  n10 = Subtract(A=Multiply(A=@n5.G, B=@n7), B=Multiply(A=@n5.R, B=@n8))
  n11 = Multiply(A=@n6.Sheen Rotation) {constB=4}
  n12 = Abs(None=ConstantBiasScale(None=Frac(None=@n11)) {bias=-0.5, scale=2})
  n13 = Constant() {r=2}
  n14 = Multiply(A=OneMinus(None=DotProduct(A=@n12, B=@n12)), B=Subtract(A=SquareRoot(None=@n13)))
  n15 = VertexInterpolator(VS=Divide(A=AppendVector(A=@n9, B=@n10), B=Add(A=@n14)))
  n16 = OneMinus(None=ComponentMask(None=Add(A=@n15.PS) {constB=0.5}) {g=true})
  n17 = Time()
  n18 = DynamicParameter() {defaultValue=<rgba 0 0.333 0.03 0>, paramNames [Sheen Sweep, Sheen Size, Sheen Rotation, Sheen Time]}
  n19 = LinearInterpolate(A=Frac(None=Multiply(A=@n17) {constB=0.333}) {desc=Sheen Position}, B=@n18.Sheen Sweep, Alpha=Ceil(None=ParticleRelativeTime())) {desc="Will swap to Particle Parameter\r\nwhen inside a partlce system"}
  n20 = Subtract(A=Multiply(A=Saturate(None=@n19), B=Add(A=@n18.Sheen Size)), B=@n18.Sheen Size)
  n21 = Saturate(None=Divide(A=Subtract(A=@n16, B=@n20), B=@n18.Sheen Size)) {desc=Remapped gradient}
  n22 = Saturate(None=ScalarParameter() {defaultValue=0.8, group=Sheen, parameterName=Sheen Center Point, sliderMax=1})
  n23 = Subtract(A=Saturate(None=@n21), B=@n22)
  n24 = Subtract(B=@n22)
  n25 = Divide(B=@n22)
  n26 = Max(A=Multiply(A=@n23, B=Divide(B=@n24)), B=Multiply(A=Multiply(A=@n25, B=@n23)) {constB=-1})
  n27 = Power(
   Base=OneMinus(None=Saturate(None=@n26))
   Exp=ScalarParameter() {defaultValue=6, group=Sheen, parameterName=Sheen Exponent, sliderMax=9, sliderMin=1}
  )
  n28 = ComponentMask(None=TextureCoordinate()) {r=true}
  n29 = OneMinus(None=@n21)
  n30 = Multiply(
   A=TextureSampleParameter2D(UVs=AppendVector(A=@n28, B=@n29)) {group=Sheen, parameterName=Sheen Texture, texture=/Game/Effects/Fort_Effects/Textures/Grad/T_Gradient_SharpBladeSheen}.RGB
   B=Ceil(None=Frac(None=@n21)) {desc=Textured Sheen Mask}
  )
  MaterialFunctionCall(Emissive=Multiply(
    A=StaticSwitchParameter(True=@n27, False=@n30) {group=Sheen, parameterName=Use Procedural Sheen, sortPriority=1}
    B=ComponentMask(None=@n4) {b=true, g=true, r=true}
   )) {materialFunction=&MF_FX_EmissiveAdjustments}
 ]
}
